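# Masked, lit particle material for the candy heart effects of the Teddy Bear glider, usable on Niagara and Cascade sprites.
# A 4x4 SubUV lookup with per-particle random offset and time-driven speed samples the candy heart mask and normal map.
# Mask alpha is tinted by particle color, which also feeds emissive adjustments scaled by EmissiveMult.
# Shading path switches give mobile a simpler SubUV function for both textures.
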Material /Game/Gadgets/Assets/VinderTech_GliderChute/Glider_TeddyBear/Effects/Materials/M_CandyHeart
graph {
 properties {
  bUsedWithNiagaraSprites=true
  bUsedWithParticleSprites=true
  blendMode=BLEND_Masked
  pixelDepthOffsetMode=PDOM_Legacy
  usedShadingModels=Default Lit
 }
 nodes [
  Constant() {r=1}
  Constant() {r=0.25}
  Constant()
  n1 = ParticleColor()
  MaterialFunctionCall(Emissive=Multiply(A=@n1.RGB, B=ScalarParameter() {parameterName=EmissiveMult})) {materialFunction=&MF_FX_EmissiveAdjustments}
  n2 = Constant() {r=4}
  n3 = Add(A=ParticleRandom()) {constB=0.05}
  n4 = Time()
  n5 = LinearInterpolate(Alpha=ParticleRandom()) {constA=0.5, constB=1.5}
  n6 = MaterialFunctionCall!(UseTexCoord 1, Rows=@n2, Columns=@n2, Particle Random Value=Add(A=Frac(None=@n3), B=Multiply(A=@n4, B=@n5))) {materialFunction=/Game/Effects/Fort_Effects/Materials/_Functions/MF_RandomSubUV_GPU_Sprites}
  n7 = Constant() {r=4}
  n8 = TextureObject() {texture=/Game/Gadgets/Assets/VinderTech_GliderChute/Glider_TeddyBear/Effects/Textures/T_Candyheart_Mask}
  n9 = Constant() {r=2}
  ShadingPathSwitch(
   Default=Multiply(
    A=@n1.RGB
    B=TextureSample(UVs=@n6.Result) {texture=&T_Candyheart_Mask}.A
   )
   Mobile=Multiply(
    A=MaterialFunctionCall!(U_Images=@n7, V_Images=@n7, SubUV Texture RGB=@n8, Time Speed=@n9, Manual Time Control, Time, UVs) {materialFunction=/Game/Effects/Fort_Effects/Materials/_Functions/MF_SubUV}.A
    B=@n1.RGB
   )
  ) {desc=Migrated from FeatureLevelSwitch}
  n10 = TextureSample(UVs=@n6.Result) {samplerType=SAMPLERTYPE_Normal, texture=/Game/Gadgets/Assets/VinderTech_GliderChute/Glider_TeddyBear/Effects/Textures/T_Candyheart_NM}
  n11 = TextureObject() {samplerType=SAMPLERTYPE_Normal, texture=&T_Candyheart_NM}
  ShadingPathSwitch(
   Default=@n10.RGB
   Mobile=MaterialFunctionCall(U_Images=@n7, V_Images=@n7, SubUV Texture RGB=@n11, Time=@n9) {materialFunction=&MF_SubUV}.RGB
  ) {desc=Migrated from FeatureLevelSwitch}
 ]
}
