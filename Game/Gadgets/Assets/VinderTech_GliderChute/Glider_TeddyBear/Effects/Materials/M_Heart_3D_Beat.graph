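# Masked lit material for mesh particles that renders a beating 3D heart in the Teddy Bear glider effects.
# Emissive combines a faked glass reflection texture and a tinted Fresnel with particle color, scaled by particle alpha.
# Radially mapped, panning swirl noise plus particle alpha is floored into a hard dissolve mask.
# A switchable world position offset pulses vertices about the object position with a sine driven by Beat and BeatSpeed dynamic parameters.
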Material /Game/Gadgets/Assets/VinderTech_GliderChute/Glider_TeddyBear/Effects/Materials/M_Heart_3D_Beat
graph {
 properties {
  bUsedWithMeshParticles=true
  bUsedWithNiagaraMeshParticles=true
  blendMode=BLEND_Masked
  pixelDepthOffsetMode=PDOM_Legacy
  usedShadingModels=Default Lit
 }
 nodes [
  n1 = ParticleColor()
  n2 = MaterialFunctionCall() {materialFunction=&VectorToRadialValue}
  n3 = Constant2Vector() {g=1.25, r=1}
  n4 = TextureSample(UVs=Panner(Coordinate=Multiply(A=@n2.Radial Coordinates, B=@n3)) {speedX=0.2, speedY=-1.5}) {samplerType=SAMPLERTYPE_LinearGrayscale, texture=/Game/Global/Textures/Default/Noise/T_SwirlNoise_N_height}
  Clamp(None=Floor(None=Add(A=@n1.A, B=@n4.R)))
  LinearInterpolate(Alpha=@n1.A) {constA=2, constB=0}
  n5 = ReflectionVectorWS()
  n6 = Multiply(A=TextureSample(UVs=ComponentMask(None=@n5) {g=true, r=true}) {texture=/Game/Items/Art_noLOD/Textures/T_Items_MineralPowder_Glass_Reflection}.RGB) {constB=0.1}
  n7 = Fresnel() {baseReflectFraction=0}
  n8 = ParticleColor()
  n9 = Multiply(A=VectorParameter() {parameterName=FresnelColor, defaultValue=<rgba 0.4213 0.2344 1 1>}.RGB, B=@n8.RGB)
  n10 = Multiply(A=Add(A=Multiply(A=@n7, B=@n9), B=Multiply(A=@n8.RGB) {constB=0.9})) {constB=0.1}
  MaterialFunctionCall(Emissive=Multiply(A=Add(A=@n6, B=@n10), B=@n1.A)) {materialFunction=&MF_FX_EmissiveAdjustments}
  ScalarParameter() {parameterName=Metallic}
  ScalarParameter() {defaultValue=1, parameterName=Specular}
  n11 = Subtract(A=ObjectPositionWS(), B=WorldPosition().XYZ)
  n12 = DynamicParameter() {paramNames [Beat,BeatSpeed,Param3,Param4]}
  n13 = LinearInterpolate(Alpha=MaterialFunctionCall(Speed=Multiply(A=@n12.BeatSpeed) {constB=2}) {materialFunction=&DebugTimeSine}.Result) {constA=-0.05, constB=0.05}
  StaticSwitchParameter(True=Multiply(A=Multiply(A=@n11, B=@n13) {constB=0.2}, B=@n12.Beat), False=ScalarParameter() {parameterName=NoWPO}) {defaultValue=true, parameterName=WPO}
 ]
}
