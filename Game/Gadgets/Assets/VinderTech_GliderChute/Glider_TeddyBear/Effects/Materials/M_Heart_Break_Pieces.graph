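# Masked, lit particle material for broken candy-heart pieces in the teddy bear glider's effects, usable on Niagara and Cascade sprites.
# The Break dynamic parameter shifts the UVs sideways and blends between two break-mask channels, selecting a piece within the heart mask.
# Particle alpha plus panning swirl noise, floored, gives a hard-edged dissolve.
# Particle-tinted Fresnel color and a faked glass reflection feed the emissive, alongside a heart normal map.
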Material /Game/Gadgets/Assets/VinderTech_GliderChute/Glider_TeddyBear/Effects/Materials/M_Heart_Break_Pieces
graph {
 properties {
  bUsedWithNiagaraSprites=true
  bUsedWithParticleSprites=true
  blendMode=BLEND_Masked
  pixelDepthOffsetMode=PDOM_Legacy
  usedShadingModels=Default Lit
 }
 nodes [
  n1 = ParticleColor()
  LinearInterpolate(Alpha=@n1.A) {constA=2, constB=0}
  n2 = ReflectionVectorWS()
  n3 = Multiply(A=TextureSample(UVs=ComponentMask(None=@n2) {g=true, r=true}) {texture=/Game/Items/Art_noLOD/Textures/T_Items_MineralPowder_Glass_Reflection}.RGB) {constB=0.1}
  n4 = Fresnel() {baseReflectFraction=0}
  n5 = ParticleColor()
  n6 = Multiply(A=VectorParameter() {parameterName=FresnelColor, defaultValue=<rgba 0.4528 0.276 1 1>}.RGB, B=@n5.RGB)
  n7 = Multiply(A=Add(A=Multiply(A=@n4, B=@n6), B=Multiply(A=@n5.RGB) {constB=0.9})) {constB=0.1}
  MaterialFunctionCall(Emissive=Multiply(A=Add(A=@n3, B=@n7), B=@n1.A)) {materialFunction=&MF_FX_EmissiveAdjustments}
  n8 = Panner() {speedX=0.2, speedY=-1.5}
  n9 = Floor(None=Add(
    A=@n1.A
    B=TextureSample(UVs=@n8) {samplerType=SAMPLERTYPE_LinearGrayscale, texture=/Game/Global/Textures/Default/Noise/T_SwirlNoise_N_height}.R
   ))
  n10 = Constant2Vector() {r=0.05}
  n11 = DynamicParameter() {paramNames [Break,Param2,Param3,Param4]}
  n12 = Add(A=TextureCoordinate(), B=LinearInterpolate(A=@n10, B=Multiply(A=@n10) {constB=-1}, Alpha=@n11.Break))
  n13 = Multiply(A=@n12) {constB=0.25}
  n14 = TextureSample(UVs=@n12) {texture=/Game/Gadgets/Assets/VinderTech_GliderChute/Glider_TeddyBear/Effects/Textures/T_Break_Mask}
  Multiply(
   A=Multiply(
    A=Clamp(None=@n9)
    B=TextureSample(UVs=@n13) {texture=/Game/Gadgets/Assets/VinderTech_GliderChute/Glider_TeddyBear/Effects/Textures/T_Candyheart_Mask}.A
   )
   B=LinearInterpolate(A=OneMinus(None=@n14.R), B=@n14.G, Alpha=@n11.Break)
  )
  TextureSample(UVs=@n13) {samplerType=SAMPLERTYPE_Normal, texture=/Game/Gadgets/Assets/VinderTech_GliderChute/Glider_TeddyBear/Effects/Textures/T_Candyheart_NM}
 ]
}
