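# Translucent, lit, two-sided particle material for a heart-shaped smoke ring effect on the teddy bear glider.
# Particle alpha and the heart texture's red channel drive a dissolve through panning swirl noise, offset by a dynamic Random parameter.
# Particle color blends with time-of-day highlight tints, while the green channel adds a Glow-scaled term, also fed to emissive adjustments.
# A quality switch applies depth fade to opacity only on High and Epic.
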Material /Game/Gadgets/Assets/VinderTech_GliderChute/Glider_TeddyBear/Effects/Materials/M_Smoke_Ring_Heart
graph {
 properties {
  bUsedWithMeshParticles=true
  bUsedWithNiagaraMeshParticles=true
  bUsedWithNiagaraSprites=true
  blendMode=BLEND_Translucent
  pixelDepthOffsetMode=PDOM_Legacy
  twoSided=true
  usedShadingModels=Default Lit
 }
 nodes [
  n1 = VectorParameter() {group=Morning Smoke, parameterName=Morning Highlights Smoke, defaultValue=<rgba 0.7183 0.5222 0.3663 0>}
  n2 = VectorParameter() {group=Daytime Smoke, parameterName=Daytime Highlights Smoke, defaultValue=<rgba 0.3103 0.3962 0.6 0>}
  n3 = VectorParameter() {group=Evening Smoke, parameterName=Evening Highlights Smoke, defaultValue=<rgba 0.64 0.4203 0.3648 0>}
  n4 = VectorParameter() {group=Nighttime Smoke, parameterName=Nighttime Highlights Smoke, defaultValue=<rgba 0.1955 0.4091 0.4894 0>}
  n5 = ParticleColor()
  n6 = LinearInterpolate(B=@n5.RGB) {constA=0.7, constAlpha=0.7}
  n7 = TextureSample() {texture=/Game/Gadgets/Assets/VinderTech_GliderChute/Glider_TeddyBear/Effects/Textures/T_Smoke_Ring_Heart}
  n8 = LinearInterpolate(Alpha=Multiply(A=@n5.A, B=@n7.R)) {constA=-1, constB=2}
  n9 = DynamicParameter() {defaultValue=<rgba 1 0 1 1>, paramNames [Random,Glow,Param3,Param4]}
  n10 = Panner(Coordinate=TextureCoordinate() {uTiling=1.5, vTiling=1.5}, Time=Add(A=@n9.Random, B=Time())) {speedX=0.15, speedY=0.7}
  n11 = Clamp(None=Multiply(
    A=Add(A=@n8, B=@n7.R)
    B=TextureSample(UVs=@n10) {samplerType=SAMPLERTYPE_LinearGrayscale, texture=/Game/Global/Textures/Default/Noise/T_SwirlNoise_N_height}.R
   ))
  n12 = LinearInterpolate(
   A=Multiply(
    A=MaterialFunctionCall(Color_Morning=@n1.RGB, Color_Day=@n2.RGB, Color_Evening=@n3.RGB, Color_Night=@n4.RGB) {materialFunction=&Color_Time_of_Day}.Result
    B=@n6
   )
   B=@n6
   Alpha=Clamp(None=Multiply(A=@n7.R, B=@n11)) {maxDefault=0.8}
  )
  n13 = Multiply(A=Multiply(A=@n5.RGB, B=@n9.Glow), B=@n7.G)
  Add(A=@n12, B=@n13)
  MaterialFunctionCall() {materialFunction=&DebugTimeSine}
  MaterialFunctionCall(Emissive=Add(A=Multiply(A=@n12) {constB=0.2}, B=@n13)) {materialFunction=&MF_FX_EmissiveAdjustments}
  n14 = DepthFade(Opacity=@n11, FadeDistance=ScalarParameter() {parameterName=DepthFade})
  QualitySwitch(Default=@n11, Low=@n11, High=@n14, Medium=@n11, Epic=@n14)
 ]
}
